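# Material function for the MirageHike female player body that layers animated swirl color onto incoming material attributes.
# A panned noise offsets UVs for distorted texture sampling, while time-driven sines phased by a grayscale swirl map index curve atlas rows for warm and cold colors.
# Named reroutes carry these colors and an emissive blend.
# The SwirlsFX_Mask red channel limits changes to base color, emissive and normal flattening.
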MaterialFunction /Game/Characters/Player/Female/Medium/Bodies/F_MED_MirageHike/Materials/MF_MirageSwirls_color
graph {nodes [
  n1 = GetMaterialAttributes(None=FunctionInput() {inputType=FunctionInput_MaterialAttributes}) {attributeGetTypes [#1, #2]}
  n2 = TextureSampleParameter2D() {
   parameterName=SwirlsFX_Mask
   samplerType=SAMPLERTYPE_Masks
   texture=/Game/Characters/Player/Female/Medium/Bodies/F_MED_MirageHike/Textures/T_F_MED_MirageHike_Body_FX
  }
  n3 = LinearInterpolate(A=@n1.BaseColor, B=LinearInterpolate(A=@n1.BaseColor, B=NamedRerouteUsage()) {constAlpha=0.3}, Alpha=@n2.R)
  n4 = NamedRerouteUsage()
  n5 = NamedRerouteUsage()
  n6 = LinearInterpolate(Alpha=@n2.R)
  FunctionOutput(None=SetMaterialAttributes!(
    MaterialAttributes=@n1.MaterialAttributes
    BaseColor=LinearInterpolate(A=@n3, B=@n4, Alpha=@n2.R)
    EmissiveColor=LinearInterpolate(B=@n5, Alpha=@n2.R)
    Normal=MaterialFunctionCall(Normal=@n1.Normal, Flatness=@n6) {materialFunction=&FlattenNormal}.Result
   ) {attributeSetTypes [#1, #3, #2]})
  n7 = TextureSample(UVs=NamedRerouteUsage()) {samplerSource=SSM_Wrap_WorldGroupSettings, texture=/Game/Characters/Player/Female/Medium/Bodies/F_MED_MirageHike/Textures/T_UVtwist}
  n8 = Constant() {r=10}
  n9 = Saturate(None=Power(Base=Multiply(A=@n7.B, B=@n8) {constB=2}, Exp=Constant() {r=1}))
  n10 = Constant() {r=0.3}
  NamedRerouteDeclaration(None=LinearInterpolate(A=NamedRerouteUsage(), B=NamedRerouteUsage(), Alpha=Multiply(A=@n9, B=@n10) {constB=0.2}) {constA=1, constAlpha=0.2}) {name=Emissive color, nodeColor=<rgba 1 0.3281 0 1>}
  n11 = TextureCoordinate()
  n12 = ScalarParameter() {defaultValue=8, parameterName=uv distortion strength Tiling}
  n13 = TextureSample(UVs=Panner(Coordinate=Multiply(A=@n11, B=@n12)) {speedX=0.2}) {samplerSource=SSM_Wrap_WorldGroupSettings, samplerType=SAMPLERTYPE_Masks, texture=/Game/Global/Textures/Default/Noise/T_WrapEdgeWear_Noise}
  n14 = LinearInterpolate(A=Constant() {r=-0.5}, B=Constant() {r=1}, Alpha=@n13.R)
  n15 = LinearInterpolate(
   A=ScalarParameter() {defaultValue=0.002, parameterName=uv distortion strength}
   Alpha=TextureSample() {samplerSource=SSM_Wrap_WorldGroupSettings, texture=&T_UVtwist}.R
  ) {constB=0}
  NamedRerouteDeclaration(None=Add(A=TextureCoordinate(), B=Multiply(A=@n14, B=@n15) {constB=0.01})) {name=UVtwist, nodeColor=<rgba 0 0.2031 1 1>}
  n16 = Time()
  n17 = Constant() {r=0.2}
  n18 = Constant()
  n19 = Constant() {r=1.5816}
  n20 = TextureSample(UVs=NamedRerouteUsage()) {
   samplerSource=SSM_Wrap_WorldGroupSettings
   samplerType=SAMPLERTYPE_LinearGrayscale
   texture=/Game/Characters/Player/Female/Medium/Bodies/F_MED_MirageHike/Textures/T_F_MED_MirageHike_Swirl
  }
  n21 = Sine(None=Add(A=Multiply(A=@n16, B=@n17) {constB=0.2}, B=LinearInterpolate(A=@n18, B=@n19, Alpha=@n20.R) {constA=0.3, constB=2}))
  n22 = CurveAtlasRowParameter(CurveTime=Add(A=Divide(A=@n21)) {constB=0.5}) {
   atlas=&/MirageHike_ColorAtlas
   curve=/Game/Characters/Player/Female/Medium/Bodies/F_MED_MirageHike/Textures/CC_MirageSwirls3
   defaultValue=1
   group=MirageSwirls
   parameterName=warmcolor_Curve
  }
  NamedRerouteDeclaration(None=@n22) {name=Warm Color, nodeColor=<rgba 1 0 0.1172 1>}
  n23 = Time()
  n24 = Constant() {r=0.3}
  n25 = Multiply(A=@n20.R, B=ScalarParameter() {defaultValue=3, parameterName=color repeat}) {constB=2}
  n26 = Sine(None=Add(A=Multiply(A=@n23, B=@n24) {constB=0.2}, B=Frac(None=@n25)))
  n27 = CurveAtlasRowParameter(CurveTime=Add(A=Divide(A=@n26)) {constB=0.5}) {
   atlas=&MirageHike_ColorAtlas
   curve=/Game/Characters/Player/Female/Medium/Bodies/F_MED_MirageHike/Textures/CC_MirageSwirls4
   group=MirageSwirls
   parameterName=Curve
  }
  n28 = Desaturation(None=@n27, Fraction=Constant())
  n29 = TextureSample(UVs=NamedRerouteUsage()) {
   samplerSource=SSM_Wrap_WorldGroupSettings
   texture=/Game/Characters/Player/Female/Medium/Bodies/F_MED_MirageHike/Textures/T_F_MED_MirageHike_Body_swirl_C
  }
  NamedRerouteDeclaration(None=Multiply(A=Multiply(A=@n28, B=@n29.RGB), B=ScalarParameter() {defaultValue=1.5, parameterName=Color indensity})) {name=Cold Color, nodeColor=<rgba 0 1 0.8594 1>}
 ]}
